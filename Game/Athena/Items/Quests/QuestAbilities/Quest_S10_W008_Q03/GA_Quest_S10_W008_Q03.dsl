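; Gameplay ability for quest S10 W008 Q03 that credits damage dealt to other players once the storm circle reaches a configured phase.
; ShouldAbilityRespondToEvent passes only when the safe zone indicator's phase is at least the Storm Phase variable and the target is not the activating pawn and carries the Homebase.Class tag.
; On activation it sends the truncated event magnitude to the owning player's quest manager as a DamagePlayerCircle custom stat, then ends.
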(event Ability|EventActivateAbilityFromEvent
  (params (EventData :type Gameplay%20Event%20Data%20Structure%20%28by%20ref%29))
  (bind (_eventtag _instigator _target _optionalobject _optionalobject2 _contexthandle _instigatortags _targettags _eventmagnitude) (Utilities|Struct|BreakGameplayEventData EventData))
  (|CallParentFunction EventData)
  (Quest|SendComplexCustomStatEvent (Variables|GAVars|Data|GetOwningPlayerQuestManager) 0 "(GameplayTags=((TagName=\"Athena.Quests.DamagePlayerCircle\")))" "(GameplayTags=)" (Math|Float|Truncate _eventmagnitude))
  (Ability|EndAbility))

(fn ShouldAbilityRespondToEvent
  (params (ActorInfo :type Gameplay%20Ability%20Actor%20Info%20Structure) (Payload_EventTag :type Gameplay%20Tag%20Structure) (Payload_Instigator :type Actor%20Object%20Reference) (Payload_Target :type Actor%20Object%20Reference) (Payload_OptionalObject :type Object%20Reference) (Payload_OptionalObject2 :type Object%20Reference) (Payload_ContextHandle :type Gameplay%20Effect%20Context%20Handle%20Structure) (Payload_InstigatorTags :type Gameplay%20Tag%20Container%20Structure) (Payload_TargetTags :type Gameplay%20Tag%20Container%20Structure) (Payload_EventMagnitude :type Float%20%28single-precision%29) (Payload_TargetData :type Gameplay%20Ability%20Target%20Data%20Handle%20Structure) (Payload_InstancedEventData :type Instanced%20Struct%20Structure))
  (bind _returnvalue (SafeZone|GetSafeZoneIndicator))
  (Utilities|IsValid _returnvalue
    (:Is%20Valid
      (if (>= (SafeZone|GetCurrentPhaseIndex _returnvalue) (Variables|Default|GetStorm%20Phase))
        (if (Utilities|NotEqual(Object) Payload_Target (Ability|GetActivatingPawn))
          (if (GameplayTags|HasTag Payload_TargetTags "(TagName=\"Homebase.Class\")")
            (return true)
            (else
              _))
          (else
            _))
        (return false)
        (else
          _)))
    (:Is%20Not%20Valid
      _)))
